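# Sound cue for the bounce impact of a severed tentacle in the PistolCanary event's post-narrative assets.
# It layers a small junk impact with one of two randomly chosen rubber impacts pitched well up, then varies the mix through a root modulator.
# It plays in the impacts sound class with small-object hit attenuation, limited to one instance by a concurrency override and with virtualization disabled.
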SoundCue /PistolCanary_Event_Audio/Assets/PostNarrative/SC_PistolCanary_SeveredTentacle_Bounce
root=Modulator_2
nodes [
 {id=Mixer_0, children [Modulator_1,WavePlayer_6], properties {inputVolume [1,1]}}
 {id=Modulator_1, children [Random_31], properties {pitchMax=1.6, pitchMin=1.5}}
 {id=Modulator_2, children [Mixer_0], properties {pitchMax=1.02}}
 {id=Random_31, children [WavePlayer_92, WavePlayer_7], properties {weights [1,1]}}
 {id=WavePlayer_6, properties {soundWaveAssetPtr=&/Junk_ImpactShort_Small03}}
 {id=WavePlayer_7, properties {soundWaveAssetPtr=&/Rubber_Impact01}}
 {id=WavePlayer_92, properties {soundWaveAssetPtr=&/Rubber_Impact02}}
]
properties {
 attenuationSettings=/Game/Sounds/Attenuation/Fort_Demolition_Hit_Small_Sized_Object_Att
 bOverrideConcurrency=true
 maxDistance=4000
 soundClassObject=&Fort_Impacts
 virtualizationMode=Disabled
 volumeMultiplier=0.6
 concurrencyOverrides {maxCount=1}
}
