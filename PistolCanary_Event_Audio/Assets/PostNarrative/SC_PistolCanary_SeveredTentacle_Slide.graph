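# Sound cue for a severed tentacle sliding in the PistolCanary event, stored with its post-narrative audio assets.
# A single looping wave player, reusing a mud knee-slide loop, feeds the output directly and is primed on load.
# It borrows pickaxe-impact attenuation, the player impact sound class and melee-weapon concurrency.
# It also stores attenuation and concurrency override values, though no flag enabling them is shown.
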SoundCue /PistolCanary_Event_Audio/Assets/PostNarrative/SC_PistolCanary_SeveredTentacle_Slide
root=WavePlayer_5
nodes [{id=WavePlayer_5, properties {bLooping=true, soundWaveAssetPtr=&/KneeSlide_Mud_Loop_Base_03}}]
properties {
 attenuationSettings=/Game/Sounds/Attenuation/Fort_Pickaxe_Impact_Att
 bPrimeOnLoad=true
 maxDistance=4300
 soundClassObject=/Game/Sounds/FortSoundClassesAndModes/Fort_Impact_Player_PSM
 volumeMultiplier=1
 attenuationOverrides {
  bAttenuateWithLPF=true
  distanceAlgorithm=Logarithmic
  falloffDistance=800
  lPFRadiusMax=800
  lPFRadiusMin=400
  attenuationShapeExtents {x=300}
 }
 concurrencyOverrides {bLimitToOwner=true, maxCount=2, resolutionRule=StopOldest}
 concurrencySet [&Player_Melee_Weapon_Conc]
}
